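# Physics asset for the FNCS Season 32 pickaxe cosmetic, a two-handed Fortnite melee weapon mesh.
# It holds a single capsule body on the Melee_TwoHanded_Handle bone, offset along the handle to wrap the weapon as one simple collision volume.
# It has no constraints, so it defines no jointed or swinging parts.
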PhysicsAsset /BRCosmetics/Weapons/FORT_Melee/Pickaxe_FNCS_S32/Meshes/FNCS_S32_Axe_PhysicsAsset
bodies {Melee_TwoHanded_Handle [{shapeType=Capsule, length=61.901, radius=20.8601, center=<0.0036 0.5049 33.0969>, rotation=<rot -0.9392 -90 0.0079>}]}
constraints []
